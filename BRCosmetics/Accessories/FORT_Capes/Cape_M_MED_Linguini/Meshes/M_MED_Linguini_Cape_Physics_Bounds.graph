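# Physics asset for the M_MED_Linguini cape back accessory, named as a bounds asset for the cape mesh.
# A capsule on backpack_root anchors five cape columns (center, mid-left, left, mid-right, right), each with identical capsules on odd-numbered dyn_cape bones 01 to 07.
# Constraints chain each column down from the root with no links between columns.
# All swing and twist motions are Free, so the 45-degree limit values do not restrict rotation.
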PhysicsAsset /BRCosmetics/Accessories/FORT_Capes/Cape_M_MED_Linguini/Meshes/M_MED_Linguini_Cape_Physics_Bounds
bodies {
 backpack_root [{shapeType=Capsule, length=17.3068, radius=8.5035, center=<0.4218 0.0872 -7.1424>, rotation=<rot 90 0 0>}]
 dyn_cape_01 [{shapeType=Capsule, length=7.5, radius=6, shapeName=dyn_cape_01_capsule, center=<6.5 0 2.85>, rotation=<rot 90 0 180>}]
 dyn_cape_01_l [{shapeType=Capsule, length=7.5, radius=6, shapeName=dyn_cape_01_l_capsule, center=<6.5 0 2.85>, rotation=<rot 90 0 180>}]
 dyn_cape_01_mid_l [{shapeType=Capsule, length=7.5, radius=6, shapeName=dyn_cape_01_mid_l_capsule, center=<6.5 0 2.85>, rotation=<rot 90 0 180>}]
 dyn_cape_01_mid_r [{shapeType=Capsule, length=7.5, radius=6, shapeName=dyn_cape_01_mid_r_capsule, center=<6.5 0 2.85>, rotation=<rot 90 0 180>}]
 dyn_cape_01_r [{shapeType=Capsule, length=7.5, radius=6, shapeName=dyn_cape_01_r_capsule, center=<6.5 0 2.85>, rotation=<rot 90 0 180>}]
 dyn_cape_03 [{shapeType=Capsule, length=7.5, radius=6, shapeName=dyn_cape_03_capsule, center=<6.5 0 2.85>, rotation=<rot 90 0 180>}]
 dyn_cape_03_l [{shapeType=Capsule, length=7.5, radius=6, shapeName=dyn_cape_03_l_capsule, center=<6.5 0 2.85>, rotation=<rot 90 0 180>}]
 dyn_cape_03_mid_l [{shapeType=Capsule, length=7.5, radius=6, shapeName=dyn_cape_03_mid_l_capsule, center=<6.5 0 2.85>, rotation=<rot 90 0 180>}]
 dyn_cape_03_mid_r [{shapeType=Capsule, length=7.5, radius=6, shapeName=dyn_cape_03_mid_r_capsule, center=<6.5 0 2.85>, rotation=<rot 90 0 180>}]
 dyn_cape_03_r [{shapeType=Capsule, length=7.5, radius=6, shapeName=dyn_cape_03_r_capsule, center=<6.5 0 2.85>, rotation=<rot 90 0 180>}]
 dyn_cape_05 [{shapeType=Capsule, length=7.5, radius=6, shapeName=dyn_cape_05_capsule, center=<6.5 0 2.85>, rotation=<rot 90 0 180>}]
 dyn_cape_05_l [{shapeType=Capsule, length=7.5, radius=6, shapeName=dyn_cape_05_l_capsule, center=<6.5 0 2.85>, rotation=<rot 90 0 180>}]
 dyn_cape_05_mid_l [{shapeType=Capsule, length=7.5, radius=6, shapeName=dyn_cape_05_mid_l_capsule, center=<6.5 0 2.85>, rotation=<rot 90 0 180>}]
 dyn_cape_05_mid_r [{shapeType=Capsule, length=7.5, radius=6, shapeName=dyn_cape_05_mid_r_capsule, center=<6.5 0 2.85>, rotation=<rot 90 0 180>}]
 dyn_cape_05_r [{shapeType=Capsule, length=7.5, radius=6, shapeName=dyn_cape_05_r_capsule, center=<6.5 0 2.85>, rotation=<rot 90 0 180>}]
 dyn_cape_07 [{shapeType=Capsule, length=7.5, radius=6, shapeName=dyn_cape_07_capsule, center=<6.5 0 2.85>, rotation=<rot 90 0 180>}]
 dyn_cape_07_l [{shapeType=Capsule, length=7.5, radius=6, shapeName=dyn_cape_07_l_capsule, center=<6.5 0 2.85>, rotation=<rot 90 0 180>}]
 dyn_cape_07_mid_l [{shapeType=Capsule, length=7.5, radius=6, shapeName=dyn_cape_07_mid_l_capsule, center=<6.5 0 2.85>, rotation=<rot 90 0 180>}]
 dyn_cape_07_mid_r [{shapeType=Capsule, length=7.5, radius=6, shapeName=dyn_cape_07_mid_r_capsule, center=<6.5 0 2.85>, rotation=<rot 90 0 180>}]
 dyn_cape_07_r [{shapeType=Capsule, length=7.5, radius=6, shapeName=dyn_cape_07_r_capsule, center=<6.5 0 2.85>, rotation=<rot 90 0 180>}]
}
constraints [
 {
  bone1Name=dyn_cape_01
  bone2Name=backpack_root
  swing1LimitDegrees=45
  swing1Motion=Free
  swing2LimitDegrees=45
  swing2Motion=Free
  twistLimitDegrees=45
  twistMotion=Free
 }
 {
  bone1Name=dyn_cape_03
  bone2Name=dyn_cape_01
  swing1LimitDegrees=45
  swing1Motion=Free
  swing2LimitDegrees=45
  swing2Motion=Free
  twistLimitDegrees=45
  twistMotion=Free
 }
 {
  bone1Name=dyn_cape_05
  bone2Name=dyn_cape_03
  swing1LimitDegrees=45
  swing1Motion=Free
  swing2LimitDegrees=45
  swing2Motion=Free
  twistLimitDegrees=45
  twistMotion=Free
 }
 {
  bone1Name=dyn_cape_07
  bone2Name=dyn_cape_05
  swing1LimitDegrees=45
  swing1Motion=Free
  swing2LimitDegrees=45
  swing2Motion=Free
  twistLimitDegrees=45
  twistMotion=Free
 }
 {
  bone1Name=dyn_cape_01_mid_l
  bone2Name=backpack_root
  swing1LimitDegrees=45
  swing1Motion=Free
  swing2LimitDegrees=45
  swing2Motion=Free
  twistLimitDegrees=45
  twistMotion=Free
 }
 {
  bone1Name=dyn_cape_03_mid_l
  bone2Name=dyn_cape_01_mid_l
  swing1LimitDegrees=45
  swing1Motion=Free
  swing2LimitDegrees=45
  swing2Motion=Free
  twistLimitDegrees=45
  twistMotion=Free
 }
 {
  bone1Name=dyn_cape_05_mid_l
  bone2Name=dyn_cape_03_mid_l
  swing1LimitDegrees=45
  swing1Motion=Free
  swing2LimitDegrees=45
  swing2Motion=Free
  twistLimitDegrees=45
  twistMotion=Free
 }
 {
  bone1Name=dyn_cape_07_mid_l
  bone2Name=dyn_cape_05_mid_l
  swing1LimitDegrees=45
  swing1Motion=Free
  swing2LimitDegrees=45
  swing2Motion=Free
  twistLimitDegrees=45
  twistMotion=Free
 }
 {
  bone1Name=dyn_cape_01_l
  bone2Name=backpack_root
  swing1LimitDegrees=45
  swing1Motion=Free
  swing2LimitDegrees=45
  swing2Motion=Free
  twistLimitDegrees=45
  twistMotion=Free
 }
 {
  bone1Name=dyn_cape_03_l
  bone2Name=dyn_cape_01_l
  swing1LimitDegrees=45
  swing1Motion=Free
  swing2LimitDegrees=45
  swing2Motion=Free
  twistLimitDegrees=45
  twistMotion=Free
 }
 {
  bone1Name=dyn_cape_05_l
  bone2Name=dyn_cape_03_l
  swing1LimitDegrees=45
  swing1Motion=Free
  swing2LimitDegrees=45
  swing2Motion=Free
  twistLimitDegrees=45
  twistMotion=Free
 }
 {
  bone1Name=dyn_cape_07_l
  bone2Name=dyn_cape_05_l
  swing1LimitDegrees=45
  swing1Motion=Free
  swing2LimitDegrees=45
  swing2Motion=Free
  twistLimitDegrees=45
  twistMotion=Free
 }
 {
  bone1Name=dyn_cape_01_mid_r
  bone2Name=backpack_root
  swing1LimitDegrees=45
  swing1Motion=Free
  swing2LimitDegrees=45
  swing2Motion=Free
  twistLimitDegrees=45
  twistMotion=Free
 }
 {
  bone1Name=dyn_cape_03_mid_r
  bone2Name=dyn_cape_01_mid_r
  swing1LimitDegrees=45
  swing1Motion=Free
  swing2LimitDegrees=45
  swing2Motion=Free
  twistLimitDegrees=45
  twistMotion=Free
 }
 {
  bone1Name=dyn_cape_05_mid_r
  bone2Name=dyn_cape_03_mid_r
  swing1LimitDegrees=45
  swing1Motion=Free
  swing2LimitDegrees=45
  swing2Motion=Free
  twistLimitDegrees=45
  twistMotion=Free
 }
 {
  bone1Name=dyn_cape_07_mid_r
  bone2Name=dyn_cape_05_mid_r
  swing1LimitDegrees=45
  swing1Motion=Free
  swing2LimitDegrees=45
  swing2Motion=Free
  twistLimitDegrees=45
  twistMotion=Free
 }
 {
  bone1Name=dyn_cape_01_r
  bone2Name=backpack_root
  swing1LimitDegrees=45
  swing1Motion=Free
  swing2LimitDegrees=45
  swing2Motion=Free
  twistLimitDegrees=45
  twistMotion=Free
 }
 {
  bone1Name=dyn_cape_03_r
  bone2Name=dyn_cape_01_r
  swing1LimitDegrees=45
  swing1Motion=Free
  swing2LimitDegrees=45
  swing2Motion=Free
  twistLimitDegrees=45
  twistMotion=Free
 }
 {
  bone1Name=dyn_cape_05_r
  bone2Name=dyn_cape_03_r
  swing1LimitDegrees=45
  swing1Motion=Free
  swing2LimitDegrees=45
  swing2Motion=Free
  twistLimitDegrees=45
  twistMotion=Free
 }
 {
  bone1Name=dyn_cape_07_r
  bone2Name=dyn_cape_05_r
  swing1LimitDegrees=45
  swing1Motion=Free
  swing2LimitDegrees=45
  swing2Motion=Free
  twistLimitDegrees=45
  twistMotion=Free
 }
]
